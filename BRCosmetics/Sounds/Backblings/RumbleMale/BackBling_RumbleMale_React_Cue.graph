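# Sound cue for the RumbleMale back bling's react sound, built on the reactive cosmetic back bling audio property sheet.
# A single wave player feeds one sound wave straight to the output at reduced volume in the back bling sound class.
# A close-range attenuation asset shapes its falloff.
# Overridden concurrency allows one instance per owner and blocks new plays while one is already playing.
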SoundCue /BRCosmetics/Sounds/Backblings/RumbleMale/BackBling_RumbleMale_React_Cue
root=WavePlayer_0
nodes [{id=WavePlayer_0, properties {soundWaveAssetPtr=&/BackBling_RumbleMale_React_01}}]
properties {
 attenuationSettings=&/Fort_BackBling_RumbleMale_Close_Athena_Att
 audioPropertiesSheet=/BRCosmetics/Sounds/Settings/PropertySheets/Reactive/APS_Cosmetic_Reactive_Backbling
 bOverrideConcurrency=true
 maxDistance=1800.0001
 soundClassObject=/Game/Sounds/FortSoundClassesAndModes/Fort_Backblings
 volumeMultiplier=0.5
 concurrencyOverrides {bLimitToOwner=true, maxCount=1, resolutionRule=PreventNew}
}
